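; Procedural jungle tile actor that scatters seeded instanced meshes along dirt-road landscape splines within its bounds.
; Construction script snaps it to a streaming sub-grid, flags same-location duplicates, runs the spawner and sets instances Static.
; The spawner copies each road spline-mesh segment into a temporary spline, then places clustered center and edge instances in three HISMs via ground traces.
; UpdateProceduralMeshPlacement also manages a location-tagged companion meshes actor.
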(event EventBeginPlay)

(event Collision|EventActorBeginOverlap
  (params (OtherActor :type Actor%20Object%20Reference)))

(event EventTick
  (params (DeltaSeconds :type Float%20%28single-precision%29)))

(event Custom|UpdateProceduralMeshPlacement
  (CallFunction|ProceduralSpawner)
  (bind _outactors (Actor|GetAllActorswithTag (Utilities|String|StringToName (Variables|Default|GetTagID))))
  (Utilities|FlowControl|Sequence
    (:then_0
      (if (> (Utilities|Array|LastIndex _outactors) 0)
        (for _array_element _outactors
          (Actor|DestroyActor _array_element))))
    (:then_1
      (if (<= (Utilities|Array|Length _outactors) 0)
        (bind _returnvalue (Game|SpawnActorfromClass "/Game/ArtTools/Procedural/Jungle/BP_ProduralJungleRoads_Meshes.BP_ProduralJungleRoads_Meshes_C" (Math|Transform|MakeTransform (- (Transformation|GetActorLocation) "0.000000,0.000000,500.000000")) "Undefined" "MultiplyWithRoot" 0 0))
        (Variables|/Script/Engine.Actor|SetTags (Utilities|Array|MakeArray (Utilities|String|StringToName (Variables|Default|GetTagID))) _returnvalue)))))

(fn AddInstancestoHISM
  (params (Spline :type Spline%20Component%20Object%20Reference) (HISM :type Instanced%20Static%20Mesh%20Component%20Object%20Reference) (SpawnChance :type Float%20%28double-precision%29 :default 1.0) (SpawnDistance :type Float%20%28double-precision%29 :default 2e+02) (AngleMin :type Float%20%28single-precision%29) (AngleMax :type Float%20%28single-precision%29) (ScaleMin :type Float%20%28single-precision%29 :default 0.75) (ScaleMax :type Float%20%28single-precision%29 :default 1.25) (ScaleEdgeAdd :type Float%20%28double-precision%29 :default 0.5) (MaxPositionOffset :type Float%20%28single-precision%29 :default 5e+01) (MaxClustering :type Integer :default 3) (Min%20Clustering :type Integer) (MaxClusteringDistance :type Float%20%28single-precision%29 :default 5e+01) (NonUniformScale :type Boolean) (ZScale%20Min :type Float%20%28single-precision%29) (ZScale%20Max :type Float%20%28single-precision%29) (ZOffset :type Float%20%28single-precision%29) (SpawnCenter :type Boolean :default true) (SpawnEdges :type Boolean :default true))
  (bind _returnvalue (Spline|GetSplineLength Spline))
  (bind _returnvalue_2 (/ _returnvalue SpawnDistance))
  (bind _returnvalue_3 (Math|Float|Truncate _returnvalue_2))
  (bind _returnvalue_4 (/ _returnvalue _returnvalue_3))
  (for _index (range _returnvalue_3)
    (if (<= (Math|Random|RandomFloatinRangefromStream (Variables|Default|GetStream) 0.0 1.0) SpawnChance)
      (for _index_2 (range (+ (Math|Random|RandomIntegerinRangefromStream (Variables|Default|GetStream) Min%20Clustering MaxClustering) 1))
        (if (not (or (> (+ (* _returnvalue_4 _index) (* _index_2 MaxClusteringDistance)) _returnvalue) (< (+ (* _returnvalue_4 _index) (* _index_2 MaxClusteringDistance)) 0)))
          (Utilities|FlowControl|Sequence
            (:then_0
              (if SpawnCenter
                (CallFunction|AddInstancewithOffset (+ (* _returnvalue_4 _index) (* _index_2 MaxClusteringDistance)) Spline HISM AngleMin AngleMax ScaleMin ScaleMax 0.0 MaxPositionOffset 0.0 NonUniformScale ZScale%20Min ZScale%20Max ZOffset)))
            (:then_1
              (if SpawnEdges
                (CallFunction|AddInstancewithOffset (+ (* _returnvalue_4 _index) (* _index_2 MaxClusteringDistance)) Spline HISM AngleMin AngleMax ScaleMin ScaleMax ScaleEdgeAdd MaxPositionOffset 2e+02 NonUniformScale ZScale%20Min ZScale%20Max ZOffset)
                (CallFunction|AddInstancewithOffset (+ (* _returnvalue_4 _index) (* _index_2 MaxClusteringDistance)) Spline HISM AngleMin AngleMax ScaleMin ScaleMax ScaleEdgeAdd MaxPositionOffset -2e+02 NonUniformScale ZScale%20Min ZScale%20Max ZOffset)))))))))

(fn AddInstanceWithOffset
  (params (Distance :type Float%20%28single-precision%29) (Spline :type Spline%20Component%20Object%20Reference) (HISM :type Instanced%20Static%20Mesh%20Component%20Object%20Reference) (AngleMin :type Float%20%28single-precision%29) (AngleMax :type Float%20%28single-precision%29) (ScaleMin2 :type Float%20%28single-precision%29) (ScaleMax2 :type Float%20%28single-precision%29) (ScaleAdd :type Float%20%28double-precision%29) (RandomOffset :type Float%20%28single-precision%29) (CenterOffset :type Float%20%28single-precision%29) (NonUniformScale :type Boolean) (ZScale%20Min :type Float%20%28single-precision%29 :default 0.75) (ZScale%20Max :type Float%20%28single-precision%29 :default 1.25) (ZOffset :type Float%20%28single-precision%29))
  (bind _returnvalue (Spline|GetLocationatDistanceAlongSpline Spline Distance "World"))
  (bind _returnvalue_2 (Spline|FindDirectionClosesttoWorldLocation Spline _returnvalue "World"))
  (bind (_origin _boxextent) (Collision|GetActorBounds self false false))
  (bind _returnvalue_3 (Math|Vector|CrossProduct _returnvalue_2 "0.000000,0.000000,1.000000"))
  (bind _returnvalue_4 (Math|Vector|NegateVector _returnvalue_3))
  (bind _stream (Variables|Default|Getstream))
  (bind _returnvalue_5 (Math|Random|RandomBoolfromStream _stream))
  (bind _returnvalue_6 (Math|Vector|SelectVector _returnvalue_3 _returnvalue_4 _returnvalue_5))
  (bind _stream_2 (Variables|Default|Getstream))
  (bind _returnvalue_7 (Math|Random|RandomFloatinRangefromStream _stream_2 0.0 RandomOffset))
  (bind _returnvalue_8 (* _returnvalue_6 _returnvalue_7))
  (bind _returnvalue_9 (+ _returnvalue _returnvalue_8))
  (bind _returnvalue_10 (* _returnvalue_3 CenterOffset))
  (bind _returnvalue_11 (+ _returnvalue_9 _returnvalue_10))
  (bind (_outhit _returnvalue_12) (Collision|LineTraceByChannel (+ _returnvalue_11 "0.000000,0.000000,10.000000") (+ _returnvalue_11 "0.000000,0.000000,-100.000000") "TraceTypeQuery10"))
  (if _returnvalue_12
    (if (Class|BTechArtGlobalLibrary|AreBoundsOverlapping? (Math|Box|MakeBox (+ (+ (Math|Vector|SelectVector (.Location (Collision|BreakHitResult _outhit)) _returnvalue_11 _returnvalue_12) (* (Math|Vector|SelectVector (.Normal (Collision|BreakHitResult _outhit)) "0.000000,0.000000,1.000000" _returnvalue_12) ZOffset)) "-0.100000,-0.100000,-0.100000") (+ (+ (Math|Vector|SelectVector (.Location (Collision|BreakHitResult _outhit)) _returnvalue_11 _returnvalue_12) (* (Math|Vector|SelectVector (.Normal (Collision|BreakHitResult _outhit)) "0.000000,0.000000,1.000000" _returnvalue_12) ZOffset)) "0.100000,0.100000,0.100000")) (Math|Box|MakeBox (+ _origin (Math|Vector|NegateVector _boxextent)) (+ _origin _boxextent)))
      (Components|InstancedStaticMesh|AddInstance HISM (Math|Transform|MakeTransform (+ (Math|Vector|SelectVector (.Location (Collision|BreakHitResult _outhit)) _returnvalue_11 _returnvalue_12) (* (Math|Vector|SelectVector (.Normal (Collision|BreakHitResult _outhit)) "0.000000,0.000000,1.000000" _returnvalue_12) ZOffset)) (Math|Rotator|CombineRotators (Math|Vector|RotatorfromAxisandAngle _returnvalue_2 (* (Math|Random|RandomFloatinRangefromStream (Variables|Default|GetStream) AngleMin AngleMax) (Math|Vector|DotProduct (Math|Vector|Normalize (- _returnvalue_9 _returnvalue)) _returnvalue_3))) (Math|Vector|RotatorfromAxisandAngle "0.000000,0.000000,1.000000" (Math|Random|RandomFloatinRangefromStream (Variables|Default|GetStream) 0.0 3.6e+02))) (+ (Math|Vector|SelectVector (Math|Vector|MakeVector (Math|Random|RandomFloatinRangefromStream (Variables|Default|GetStream) ScaleMin2 ScaleMax2) (Math|Random|RandomFloatinRangefromStream (Variables|Default|GetStream) ScaleMin2 ScaleMax2) (Math|Random|RandomFloatinRangefromStream (Variables|Default|GetStream) ZScale%20Min ZScale%20Max)) (Math|Vector|MakeVector (Math|Random|RandomFloatinRangefromStream (Variables|Default|GetStream) ScaleMin2 ScaleMax2) (Math|Random|RandomFloatinRangefromStream (Variables|Default|GetStream) ScaleMin2 ScaleMax2) (Math|Random|RandomFloatinRangefromStream (Variables|Default|GetStream) ScaleMin2 ScaleMax2)) NonUniformScale) ScaleAdd)) true))))

(fn ProceduralSpawner
  (bind (_origin _boxextent) (Collision|GetActorBounds self false false))
  (bind _tempspline (LocalVariables|GettempSpline))
  (bind _tempspline_2 (LocalVariables|GettempSpline))
  (Math|Random|SetRandomStreamSeed (Variables|Default|Getstream))
  (Components|InstancedStaticMesh|ClearInstances (Variables|Default|GetHierarchicalInstancedStaticMesh))
  (Utilities|FlowControl|Sequence
    (:then_0
      (bind _outactors (Actor|GetAllActorsOfClass "/Script/Landscape.LandscapeSplineActor"))
      (for _array_element _outactors
        (Utilities|FlowControl|Sequence
          (:then_0
            (Utilities|Array|Clear (LocalVariables|GetTransformsList))
            (bind _returnvalue (AddComponent|NODE_AddSplineComponent-0 :RelativeTransform (Math|Transform|MakeTransform "0, 0, 0")))
            (LocalVariables|SettempSpline _returnvalue)
            (for _array_element_2 (Actor|GetComponentsByClass _array_element "/Script/Engine.SplineMeshComponent")
              (if (== (Variables|/Script/Engine.StaticMeshComponent|GetStaticMesh _array_element_2) "/Game/Environments/Asteria/Terrain/DirtRoad/SM_DirtRoad.SM_DirtRoad")
                (Spline|RemoveSplinePoint _tempspline_2 0 true)
                (Spline|RemoveSplinePoint _tempspline_2 0 true)
                (if (Class|BTechArtGlobalLibrary|AreBoundsOverlapping? (Math|Box|MakeBox (+ (Collision|GetComponentBounds _array_element_2) (Math|Vector|NegateVector (.BoxExtent (Collision|GetComponentBounds _array_element_2)))) (+ (Collision|GetComponentBounds _array_element_2) (.BoxExtent (Collision|GetComponentBounds _array_element_2)))) (Math|Box|MakeBox (+ _origin (Math|Vector|NegateVector _boxextent)) (+ _origin _boxextent)))
                  (Utilities|FlowControl|Sequence
                    (:then_0
                      (Spline|AddSplinePointatIndex _tempspline (+ (Transformation|GetWorldLocation _array_element_2) (SplineMesh|GetStartPosition _array_element_2)) 0 "World" true)
                      (Spline|SetTangentatSplinePoint _tempspline 0 (SplineMesh|GetStartTangent _array_element_2) "World" true)
                      (Spline|AddSplinePointatIndex _tempspline (+ (Transformation|GetWorldLocation _array_element_2) (SplineMesh|GetEndPosition _array_element_2)) 1 "World" true)
                      (Spline|SetTangentatSplinePoint _tempspline 1 (SplineMesh|GetEndTangent _array_element_2) "World" true))
                    (:then_1
                      (Math|Random|SetRandomStreamSeed (Variables|Default|Getstream) (Variables|Default|GetSeed))
                      (CallFunction|AddInstancestoHISM (LocalVariables|GettempSpline) (Variables|Default|GetHierarchicalInstancedStaticMesh) 1.0 5e+01 0.0 0.0 0.75 1.0 0.15 2e+01 0 0 5e+01 true 1.0 1.5 -7.0 false)
                      (CallFunction|AddInstancestoHISM (LocalVariables|GettempSpline) (Variables|Default|GetHierarchicalInstancedStaticMesh) 0.8 5e+02 0.0 0.0 0.55 0.75 0.15 2e+01 15 10 5e+01 true 1.0 1.5 -7.0 true false)
                      (Math|Random|SetRandomStreamSeed (Variables|Default|Getstream) (+ (Variables|Default|GetSeed) 234))
                      (CallFunction|AddInstancestoHISM (LocalVariables|GettempSpline) (Variables|Default|GetHierarchicalInstancedStaticMesh2) 0.75 2e+02 25.0 3e+01 0.5 0.75 0.4 5e+01 2 0 5e+01 false 1.0 1.5 4.0 false)
                      (CallFunction|AddInstancestoHISM (LocalVariables|GettempSpline) (Variables|Default|GetHierarchicalInstancedStaticMesh2) 0.75 8e+02 25.0 3e+01 0.5 0.75 0.4 5e+01 2 0 5e+01 false 1.0 1.5 4.0 true false)
                      (Math|Random|SetRandomStreamSeed (Variables|Default|Getstream) (+ (Variables|Default|GetSeed) 112))
                      (CallFunction|AddInstancestoHISM (LocalVariables|GettempSpline) (Variables|Default|GetHierarchicalInstancedStaticMesh3) 1.0 2e+02 25.0 45.0 0.68 0.75 0.5 5e+01 2 0 5e+01 false 1.0 1.5 4.0 false)
                      (CallFunction|AddInstancestoHISM (LocalVariables|GettempSpline) (Variables|Default|GetHierarchicalInstancedStaticMesh3) 1.0 7e+02 25.0 45.0 0.68 0.75 0.5 25.0 3 0 25.0 false 1.0 1.5 4.0 true false))))))))))))

(fn SetInstancestoStatic
  (for _array_element (Actor|GetComponentsByClass "/Script/Engine.HierarchicalInstancedStaticMeshComponent")
    (Transformation|SetMobility _array_element "Static")))

(fn ConstructionScript
  (Utilities|FlowControl|Sequence
    (:then_0
      (Transformation|SetActorLocation (Math|Vector|MakeVector (* (- (Math|Float|Ceil (/ (.x (Transformation|GetActorLocation)) (/ (Variables|Default|GetGridSize) (Variables|Default|GetTilesPerStreamingGrid)))) 0.5) (/ (Variables|Default|GetGridSize) (Variables|Default|GetTilesPerStreamingGrid))) (* (- (Math|Float|Ceil (/ (.y (Transformation|GetActorLocation)) (/ (Variables|Default|GetGridSize) (Variables|Default|GetTilesPerStreamingGrid)))) 0.5) (/ (Variables|Default|GetGridSize) (Variables|Default|GetTilesPerStreamingGrid))) 5e+03))
      (Components|Box|SetBoxExtent (Variables|Default|GetBox) (Math|Vector|MakeVector (/ (/ (Variables|Default|GetGridSize) (Variables|Default|GetTilesPerStreamingGrid)) 2.0) (/ (/ (Variables|Default|GetGridSize) (Variables|Default|GetTilesPerStreamingGrid)) 2.0) (Variables|Default|GetVolumeHeight)) true)
      (Variables|Rendering|SetSpriteScale 5.0)
      (bind _outactors (Actor|GetAllActorsOfClass "/Game/ArtTools/Procedural/Jungle/BP_ProceduralJungleRoads.BP_ProceduralJungleRoads_C"))
      (bind _array_element (Utilities|Array|ForEachLoopwithBreak _outactors)
        (:LoopBody
          (if (!= _array_element self)
            (if (Utilities|Operators|Equal(==) (Transformation|GetActorLocation _array_element) (Transformation|GetActorLocation) 5.0)
              (Development|PrintString (Utilities|String|Append (Utilities|String|Append (Utilities|GetDisplayName _array_element) "   AND   ") (Utilities|GetDisplayName self)) true true "(R=1.000000,G=0.057065,B=0.000000,A=1.000000)")
              (Utilities|FlowControl|Sequence
                (:then_0
                  (Rendering|Components|Sprite|SetSprite (Variables|Default|GetBillboard) "/Engine/EditorResources/BadBlueprintSprite.BadBlueprintSprite")
                  (Variables|Rendering|SetSpriteScale 3e+01))
                (:then_1
                  _)))))
        (:Completed
          _))
      (Utilities|FlowControl|Gate false
        (:Exit
          (Rendering|Components|Sprite|SetSprite (Variables|Default|GetBillboard) "/Engine/EditorResources/S_TriggerBox.S_TriggerBox")
          (Variables|Rendering|SetSpriteScale 5.0))))
    (:then_1
      (Variables|Default|SetTagID (Utilities|String|Append "ProceduralActorID_" (Utilities|String|Append (Utilities|String|ToString(Integer) (Math|Float|Ceil (.x (Transformation|GetActorLocation)))) (Utilities|String|ToString(Integer) (Math|Float|Ceil (.y (Transformation|GetActorLocation))))))))
    (:then_2
      (CallFunction|ProceduralSpawner))
    (:then_3
      (CallFunction|SetInstancestoStatic))))
